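# Idle Niagara effect for the VelvetDeskFiber backpack cosmetic, layering fire meshes, flames, embers, glows and a light.
# Burst-spawned fire and glow particles persist at preset offsets, with array-placed fire meshes using three renderers and extra color and glow-scale attributes.
# System Timeline values fade these layers and the light, and ramp the continuous flame and ember spawn rates.
# Bottom fire and glow respond to time of day; bottom fire also adjusts for front-end preview.
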NiagaraSystem /BRCosmetics/Accessories/FORT_Backpacks/Backpack_M_MED_VelvetDeskFiber/FX/NS_Backpack_VelvetDeskFiber_Idle
dependencies {
 usedDataInterfaces [&NiagaraDataInterfaceCurve, &NiagaraDataInterfaceArrayFloat3, &NiagaraDataInterfaceArrayFloat]
 usedDynamicInputs [&SelectFloatFromArray]
 usedModules [&SystemState, &Timeline]
}
emitters {
 BottomFire {
  emitterUpdateScript [
   &EmitterState(Life Cycle Mode=::NewEnumerator1, Loop Behavior=::NewEnumerator1, Loop Duration Mode=::NewEnumerator1, UseLoopDelay=false)
   &SpawnBurst_Instantaneous(Use Spawn Probability=false, UseLoopCountLimit=false)
  ]
  particleSpawnScript [
   &InitializeParticle(
    Color=&MakeCustomLinearColorFromBool(Bool=$User.bIsFrontEndPreview, True Color=<rgba 2 0.106 0 0.2>, False Color=<rgba 2 0.1353 0 0.8>)
    Lifetime=5
    Mesh Scale=&Multiply_Vector(A=<0.08 0.035 0.035>, B=<0.7 1.25 1.6>)
    Mesh Scale Mode=::NewEnumerator4
    Position Offset=<0 0 -1>
    Sprite Size Mode=::NewEnumerator3
    Uniform Sprite Size=38.4
    UsePositionOffset=true
    Write Material Random=true
   )
   &InitialMeshOrientation(Mesh Orientation Mode=::NewEnumerator2, Rotation=<0 -0.25 0>, Use Rotation Vector=true)
  ]
  particleUpdateScript [
   &ParticleState(Kill Particles When Lifetime Has Elapsed=false, Let Infinitely Lived Particles Die When Emitter Deactivates=true)
   &ScaleMeshSize(Scale Factor=&Multiply_Vector(A=<1 1 1>, B=&VectorFromFloat(Value=$User.FXScale)))
   &ScaleColor(
    Scale Alpha=$System.Timeline.Time
    Scale RGB=&VectorFromFloat(Value=&FNTimeOfDayScalar(Night Scale=0.8, Morning Scale=1, Evening Scale=0.5))
    ScaleA=true
    ScaleRGB=true
   )
   &CameraOffset(Camera Offset Amount=2)
  ]
  rendererClasses [&NiagaraMeshRendererProperties]
  renderers [{rendererClass=&NiagaraMeshRendererProperties, rendererIndex=0}]
 }
 Ember {
  emitterUpdateScript [
   &EmitterState()
   &SpawnRate(
    Spawn Probability=0.6
    SpawnRate=&FloatFromCurve(
     FloatCurve=DataInterface() {propertyValues {Curve {keys [{tangentMode=RCTM_User, time=0, value=0}, {tangentMode=RCTM_User, time=1, value=1}]}}}
     CurveIndex=$System.Timeline.Time
     Scale Curve=12
    )
    Use Spawn Probability=true
   )
  ]
  particleSpawnScript [
   &InitializeParticle(
    Color=<rgba 1.4 0.7736 0.0185 1>
    Lifetime Max=0.7
    Lifetime Min=0.3
    Lifetime Mode=::NewEnumerator1
    Position Offset=<0 0 5>
    Sprite Size Mode=::NewEnumerator1
    Uniform Sprite Size Max=6
    Uniform Sprite Size Min=2
    UsePositionOffset=true
    Write Material Random=true
   )
   &ShapeLocation(MaskSpawnGroup=false, Override Seed=false, Sphere Radius=8, UseOwnerScale=true)
   &AddVelocity(Velocity=&RandomRangeVector(Minimum=<5 5 15>, Maximum=<-5 -5 55>))
  ]
  particleUpdateScript [
   &Drag(Drag=3, Use Linear Drag=true, Use Rotational Drag=true)
   &SolveForcesAndVelocity(Clamp Velocity=false, Limit Acceleration=false)
   &ParticleState()
   &ScaleSpriteSize(Uniform Curve Sprite Scale=DataInterface() {propertyValues {Curve {keys [{tangentMode=RCTM_User, time=0, value=0}, <at 0.2 1>, <at 0.6 1>, {tangentMode=RCTM_User, time=1, value=0}]}}})
   &DynamicMaterialParameters(
    Index 0 Param 1=&FloatFromCurve(FloatCurve=DataInterface() {propertyValues {Curve {keys [{interpMode=RCIM_Cubic, leaveTangent=3.2748, tangentMode=RCTM_Break, time=0.4, value=0}, <at 1 1>]}}})
    Index 0 Param 2=&RandomRangeFloat(Minimum=0, Maximum=1)
    Index 0 Param 3=&RandomRangeFloat(Minimum=0, Maximum=1)
    Module.Index0Float0Write=true
    Module.Index0Float1Write=true
    Module.Index0Float2Write=true
    Module.Index0Float3Write=true
   )
  ]
  rendererClasses [&NiagaraSpriteRendererProperties]
  renderers [{rendererClass=&NiagaraSpriteRendererProperties}]
 }
 FireLarge {
  emitterUpdateScript [
   &EmitterState(Life Cycle Mode=::NewEnumerator1, Loop Behavior=::NewEnumerator1, Loop Duration=1, UseLoopDelay=false)
   &SpawnBurst_Instantaneous(Spawn Count=3, Use Spawn Probability=false, UseLoopCountLimit=false)
  ]
  particleSpawnScript [
   &PartitionParticles(Number Of Partitions=3)
   &InitializeParticle(
    Color=<rgba 4 0.7647 0 0.2>
    Lifetime=5
    Mesh Scale=&SelectVectorFromArray(
     Direct Array Index=$Particles.PartitionParticles.Partition
     Vector Selection Array=DataInterface() {propertyValues {FloatData [<1.3 1.1737 0.7>, <1.3 1.1737 0.7>, <1.4608 1.2867 0.6649>]}}
    )
    Mesh Scale Mode=::NewEnumerator4
    Position Offset=&SelectVectorFromArray(
     Direct Array Index=$Particles.PartitionParticles.Partition
     Vector Selection Array=DataInterface() {propertyValues {FloatData [<7.7156 8.4897 -9>, <-8.0294 6.2668 -9>, <0 9.8786 -2.5>]}}
    )
    UsePositionOffset=true
    Write Material Random=true
   )
   &InitialMeshOrientation(Mesh Axis To Orient=<-1 0 0>, Mesh Orientation Mode=::NewEnumerator2, Rotation=<0 0 0.25>, Use Rotation Vector=true)
  ]
  particleUpdateScript [
   &ParticleState(Kill Particles When Lifetime Has Elapsed=false)
   &InterpolateOverTime(Rate Of Change=8, Target Vector Value:NiagaraPosition=$Particles.Position, Variable Type=::NewEnumerator2) {enabled=false}
   /SetVariables(Particles.Interpolate Dynamic=&MakeFloat4FromVectorAndFloat("Vector (RGB)"=$Particles.InterpolateOverTime.InterpolateOverTime_V3_.Moving Average, "Float (Alpha)"=0.7)) {enabled=false, isSetParameters=true, module=SetVariables_#1}
   &ScaleColor(Scale Alpha=$System.Timeline.Time, ScaleA=true, ScaleRGB=true)
   /SetVariables(Particles.Color2=&Multiply_LinearColorByFloat(Linear Color=<rgba 0.3021 0.0251 0.002 0.45>, Float=$System.Timeline.Time)) {isSetParameters=true, module=SetVariables_#2}
   /SetVariables(Particles.Color3:LinearColor=&Multiply_LinearColorByFloat(Linear Color=<rgba 2 0.3353 0.1698 0.05>, Float=$System.Timeline.Time)) {isSetParameters=true, module=SetVariables_#3}
   /SetVariables(Particles.GlowScale:Vector3f=&Multiply_Vector(A=$Particles.Initial.Scale, B=<1.2 1.2 0.96>)) {isSetParameters=true, module=SetVariables_#4}
  ]
  rendererClasses [&NiagaraMeshRendererProperties]
  renderers [
   {rendererClass=&NiagaraMeshRendererProperties, rendererIndex=0}
   {rendererClass=&NiagaraMeshRendererProperties, rendererIndex=1}
   {rendererClass=&NiagaraMeshRendererProperties, rendererIndex=2}
  ]
 }
 FireSmall {
  emitterUpdateScript [
   &EmitterState(Life Cycle Mode=::NewEnumerator1, Loop Behavior=::NewEnumerator1, Loop Duration=1, UseLoopDelay=false)
   &SpawnBurst_Instantaneous(Spawn Count=8, Use Spawn Probability=false, UseLoopCountLimit=false)
  ]
  particleSpawnScript [
   &PartitionParticles(Number Of Partitions=8)
   &InitializeParticle(
    Color=<rgba 5 0.9558 0 0.4>
    Lifetime=5
    Mesh Scale=&SelectVectorFromArray(
     Direct Array Index=$Particles.PartitionParticles.Partition
     Vector Selection Array=DataInterface() {propertyValues {FloatData [<0.4 0.4 0.4>, <0.3 0.3 0.3>, <0.4 0.4 0.4>, <0.3 0.3 0.3>, <0.5 0.5 0.5>, <-0.5 0.5 0.5>, <0.25 0.25 0.25>, <-0.25 0.25 0.25>]}}
    )
    Mesh Scale Mode=::NewEnumerator4
    Position Offset=&SelectVectorFromArray(
     Direct Array Index=$Particles.PartitionParticles.Partition
     Vector Selection Array=DataInterface() {propertyValues {FloatData [
        <-10.271 -5.8189 3.7745>
        <-6.4863 -6.6262 -0.09>
        <10.271 -5.8189 3.7745>
        <6.4863 -6.6262 -0.09>
        <10 -5 -10>
        <-10 -5 -10>
        <-13.5 -4.5 3>
        <13.5 -4.5 3>
       ]}}
    )
    UsePositionOffset=true
    Write Material Random=true
   )
   &InitialMeshOrientation(
    Mesh Axis To Orient=<-1 0 0>
    Mesh Orientation Mode=::NewEnumerator2
    Rotation=&SelectVectorFromArray(
     Direct Array Index=$Particles.PartitionParticles.Partition
     Vector Selection Array=DataInterface() {propertyValues {FloatData [<0.16 -0.09 0.5>, <0.18 -0.06 0.5>, <0 -0.1 0.138>, <0 -0.11 0>, <0.1532 -0.0811 0>, <0.15 0.0811 0>, <0 0 0>, <0 0 0>]}}
    )
    Use Rotation Vector=true
   )
  ]
  particleUpdateScript [
   &ParticleState(Kill Particles When Lifetime Has Elapsed=false)
   &InterpolateOverTime(Rate Of Change=8, Target Vector Value:NiagaraPosition=$Particles.Position, Variable Type=::NewEnumerator2) {enabled=false}
   /SetVariables(Particles.Interpolate Dynamic=&MakeFloat4FromVectorAndFloat("Vector (RGB)"=$Particles.InterpolateOverTime.InterpolateOverTime_V3_.Moving Average, "Float (Alpha)"=0.7)) {enabled=false, isSetParameters=true, module=SetVariables_#1}
   &ScaleColor(Scale Alpha=$System.Timeline.Time, ScaleA=true, ScaleRGB=true)
   /SetVariables(Particles.Color2=&Multiply_LinearColorByFloat(Linear Color=<rgba 0.3021 0.0251 0.002 0.45>, Float=$System.Timeline.Time)) {isSetParameters=true, module=SetVariables_#2}
   /SetVariables(Particles.Color3:LinearColor=&Multiply_LinearColorByFloat(Linear Color=<rgba 2 0.3353 0.1698 0.05>, Float=$System.Timeline.Time)) {isSetParameters=true, module=SetVariables_#3}
   /SetVariables(Particles.GlowScale:Vector3f=&Multiply_Vector(A=$Particles.Initial.Scale, B=<1.2 1.2 0.96>)) {isSetParameters=true, module=SetVariables_#4}
  ]
  rendererClasses [&NiagaraMeshRendererProperties]
  renderers [
   {rendererClass=&NiagaraMeshRendererProperties, rendererIndex=0}
   {rendererClass=&NiagaraMeshRendererProperties, rendererIndex=1}
   {rendererClass=&NiagaraMeshRendererProperties, rendererIndex=2}
  ]
 }
 Flame {
  emitterUpdateScript [
   &EmitterState()
   &SpawnRate(
    Spawn Probability=0.8
    SpawnRate=&FloatFromCurve(
     FloatCurve=DataInterface() {propertyValues {Curve {keys [{tangentMode=RCTM_User, time=0, value=0}, {tangentMode=RCTM_User, time=1, value=1}]}}}
     CurveIndex=$System.Timeline.Time
     Scale Curve=8
    )
    Use Spawn Probability=true
   )
  ]
  particleSpawnScript [
   &InitializeParticle(
    Lifetime Max=1.8
    Lifetime Min=0.9
    Lifetime Mode=::NewEnumerator1
    Mesh Scale=&RandomRangeVector(Minimum=<0.12 1 1>, Maximum=<0.155 1 1>)
    Mesh Scale Mode=::NewEnumerator4
    Position Offset=<0 0 -3>
    Sprite Size=<18 29>
    Sprite Size Mode=::NewEnumerator4
    UsePositionOffset=true
    Write Material Random=true
   )
   &ShapeLocation(
    MaskSpawnGroup=false
    Non Uniform Scale=<1 1 0.8>
    Override Seed=false
    Sphere Radius=10
    Sphere Surface Distribution=0.6548
    UseOwnerScale=true
   )
   &SpriteFacingAndAlignment(Alignment Coordinate Space=::Local, Sprite Alignment=<0 0 1>, WriteAlignment=true, WriteFacing=false) {enabled=false}
   &InitialMeshOrientation(Mesh Orientation Mode=::NewEnumerator2, Orientation Coordinate Space=::Local, Rotation=<0 -0.25 0>, Use Rotation Vector=true)
  ]
  particleUpdateScript [
   &SolveForcesAndVelocity(Clamp Velocity=false, Limit Acceleration=false)
   &ParticleState()
   &ScaleColor(
    Scale Alpha=&FloatFromCurve(FloatCurve=DataInterface() {propertyValues {Curve {keys [
         {arriveTangent=5.5416, interpMode=RCIM_Cubic, leaveTangent=5.5416, tangentMode=RCTM_User, time=0, value=0.1}
         {interpMode=RCIM_Cubic, leaveTangent=-0.0467, tangentMode=RCTM_Break, time=0.5, value=1.05}
         {arriveTangent=-0.2525, tangentMode=RCTM_User, time=1, value=0}
        ]}}})
    ScaleA=true
    ScaleRGB=true
   )
   &CameraOffset(Camera Offset Amount=-3)
   &DynamicMaterialParameters(
    Index 0 Param 1=1
    Module.Index0Float0Write=true
    Module.Index0Float1Write=true
    Module.Index0Float2Write=true
    Module.Index0Float3Write=true
   )
   /SetVariables(Particles.Render2DynPar:Vector4f=&MakeVector4(X=1, Y=&RandomRangeFloat(Minimum=0, Maximum=1), Z=0, W=0)) {isSetParameters=true}
  ]
  rendererClasses [&NiagaraMeshRendererProperties]
  renderers [{rendererClass=&NiagaraMeshRendererProperties, rendererIndex=0}]
 }
 Glow001 {
  emitterUpdateScript [
   &EmitterState(Life Cycle Mode=::NewEnumerator1, Loop Behavior=::NewEnumerator1, Loop Duration=1, UseLoopDelay=false)
   &SpawnBurst_Instantaneous(Spawn Count=15, Use Spawn Probability=false, UseLoopCountLimit=false) {module=SpawnBurst_Instantaneous001}
   &SpawnBurst_Instantaneous(Spawn Count=2, Spawn Group=2, Use Spawn Probability=false, UseLoopCountLimit=false) {module=SpawnBurst_Instantaneous002}
   &SpawnBurst_Instantaneous(Spawn Group=1, Use Spawn Probability=false, UseLoopCountLimit=false)
  ]
  particleSpawnScript [
   /SetVariables(Particles.VisibilityTag=&SpawnGroup()) {isSetParameters=true}
   &InitializeParticle(
    Color=&MaskLinearColorBySpawnGroup(Masked Linear Color=<rgba 1 0 0.0095 0.2>, Passthrough Linear Color=<rgba 1 0 0.0095 0.4>)
    Lifetime=5
    Mesh Scale Mode=::NewEnumerator3
    Position Offset=&MaskVectorBySpawnGroup(
     Masked Vector=&MakeVector(
      X=&Lerp_Float(
       A=8
       B=-8
       Alpha=&FloatFromCurve(
        FloatCurve=DataInterface() {propertyValues {Curve {keys [<at 0 1>,<at 1 0>]}}}
        CurveIndex=&ReturnNormalizedExecIndex()
       )
      )
      Y=-1
      Z=&Lerp_Float(
       A=-5
       B=-10
       Alpha=&FloatFromCurve(
        FloatCurve=DataInterface() {propertyValues {Curve {keys [
            {interpMode=RCIM_Cubic, leaveTangent=4.5395, tangentMode=RCTM_Break, time=0, value=0}
            {interpMode=RCIM_Cubic, tangentMode=RCTM_Break, time=0.5, value=1.05}
            {arriveTangent=-3.9091, interpMode=RCIM_Cubic, tangentMode=RCTM_Break, time=1, value=0}
           ]}}}
        CurveIndex=&ReturnNormalizedExecIndex()
       )
      )
     )
     Passthrough Vector=&MaskVectorBySpawnGroup(
      Masked Vector=<0 1 -13.5>
      Passthrough Vector=&SelectVectorFromArray(
       Direct Array Index=&ReturnExecIndex()
       Vector Selection Array=DataInterface() {propertyValues {FloatData [<-1.5 -2 -11.5>, <1.5 -2 -11.5>]}}
      )
      Spawn Group Mask=1
     )
    )
    Sprite Size Mode=::NewEnumerator3
    Uniform Sprite Size=&MaskFloatBySpawnGroup(Masked Float=5, Passthrough Float=4, Spawn Group Mask=2)
    UsePositionOffset=true
    Write Material Random=true
   )
   &DynamicMaterialParameters(
    Index 0 Param 1=1.5
    Module.Index0Float0Write=true
    Module.Index0Float1Write=true
    Module.Index0Float2Write=true
    Module.Index0Float3Write=true
   )
   &SpriteFacingAndAlignment(
    Alignment Coordinate Space=::World
    Facing Coordinate Space=::Local
    Sprite Alignment=<0 0 1>
    Sprite Facing=&MakeCustomVectorFromBool(
     Bool=&SetBoolByIntComparison(A=$Particles.PartitionParticles.Partition, B=3, Comparison Type=::NewEnumerator6)
     True Vector=<0 -1 0>
     False Vector=&MakeCustomVectorFromBool(
      Bool=&SetBoolByIntComparison(A=$Particles.PartitionParticles.Partition, B=3, Comparison Type=::NewEnumerator4)
      True Vector=<-1 0 0>
      False Vector=<1 0 0>
     )
    )
    WriteAlignment=true
    WriteFacing=true
   ) {enabled=false}
  ]
  particleUpdateScript [
   &ParticleState(Kill Particles When Lifetime Has Elapsed=false)
   &ScaleColor(
    Scale Alpha=&FloatFromCurve(
     FloatCurve=DataInterface() {propertyValues {Curve {keys [{tangentMode=RCTM_User, time=0, value=0}, {tangentMode=RCTM_User, time=1, value=1}]}}}
     CurveIndex=$System.Timeline001.Time
     Scale Curve=&FNTimeOfDayScalar(Night Scale=&MaskFloatBySpawnGroup(Masked Float=1, Passthrough Float=0.5), Morning Scale=1, Evening Scale=1)
    )
    ScaleA=true
    ScaleRGB=true
   )
   &CameraOffset(Camera Offset Amount=&MaskFloatBySpawnGroup(
     Masked Float=1
     Passthrough Float=&MaskFloatBySpawnGroup(Masked Float=1.5, Passthrough Float=0.3, Spawn Group Mask=1)
    ))
  ]
  rendererClasses [&NiagaraSpriteRendererProperties, &NiagaraMeshRendererProperties]
  renderers [
   {rendererClass=&NiagaraSpriteRendererProperties}
   {rendererClass=&NiagaraMeshRendererProperties, rendererIndex=1}
   {rendererClass=&NiagaraSpriteRendererProperties, rendererIndex=2}
  ]
 }
 LightAndGlow {
  emitterUpdateScript [
   &EmitterState(Life Cycle Mode=::NewEnumerator1, Loop Behavior=::NewEnumerator1, Loop Duration=1, UseLoopDelay=false)
   &SpawnBurst_Instantaneous(Use Spawn Probability=false, UseLoopCountLimit=false)
  ]
  particleSpawnScript [
   &InitializeParticle(Color=<rgba 2 0.1523 0 0.6>, Sprite Size Mode=::NewEnumerator3, Uniform Sprite Size=28, UsePositionOffset=false, Write Material Random=true)
   &SpriteFacingAndAlignment(Alignment Coordinate Space=::Local, Sprite Alignment=<0 0 1>, WriteAlignment=true, WriteFacing=false) {enabled=false}
   &DynamicMaterialParameters(
    Index 0 Param 1=1.6
    Index 0 Param 2=0.2
    Module.Index0Float0Write=true
    Module.Index0Float1Write=true
    Module.Index0Float2Write=true
    Module.Index0Float3Write=true
   )
   /SetVariables() {isSetParameters=true}
   /SetVariables(
    Particles.DecalRenderPos:NiagaraPosition=&AddVectorToPosition(Position=$Particles.Position, Vector=<0 0 13>)
    Particles.DecalRenderSize:Vector2f=&Vector2DFromFloat(Value=45)
   ) {enabled=false, isSetParameters=true}
   /SetVariables(Particles.DecalDynPar:Vector4f=&MakeVector4(X=2, Y=0, Z=1, W=0)) {enabled=false, isSetParameters=true}
  ]
  particleUpdateScript [
   &ParticleState(Kill Particles When Lifetime Has Elapsed=false)
   &ScaleColor(Scale Alpha=$System.Timeline.Time, ScaleA=true, ScaleRGB=true)
   &FN_Light_Attributes(
    Color=&Multiply_LinearColorByFloat(Linear Color=$Particles.Initial.Color, Float=10)
    Create Color Attribute=true
    Create Exponent Attribute=true
    Create Position Attribute=true
    Create Radius Attribute=true
    Exponent=0.5
    Position=&AddVectorToPosition(Position=$Particles.Position, Vector=<0 8 0>)
    Radius=12
    Scale Color=$System.Timeline.Time
   )
   /SetVariables(Particles.DecalCameraOffset=14) {enabled=false, isSetParameters=true}
  ]
  rendererClasses [&NiagaraSpriteRendererProperties, &NiagaraLightRendererProperties]
  renderers [
   {rendererClass=&NiagaraSpriteRendererProperties}
   {rendererClass=&NiagaraLightRendererProperties, rendererIndex=1}
   {rendererClass=&NiagaraSpriteRendererProperties, rendererIndex=2}
  ]
 }
}
summary {
 systemName=NS_Backpack_VelvetDeskFiber_Idle
 emitters [
  {emitterName=LightAndGlow, rendererClasses [&NiagaraSpriteRendererProperties, &NiagaraLightRendererProperties]}
  {emitterName=Flame, rendererClasses [&NiagaraMeshRendererProperties]}
  {emitterName=Ember, rendererClasses [&NiagaraSpriteRendererProperties]}
  {emitterName=FireSmall, rendererClasses [&NiagaraMeshRendererProperties]}
  {emitterName=FireLarge, rendererClasses [&NiagaraMeshRendererProperties]}
  {emitterName=BottomFire, rendererClasses [&NiagaraMeshRendererProperties]}
  {emitterName=Glow001, rendererClasses [&NiagaraSpriteRendererProperties, &NiagaraMeshRendererProperties]}
 ]
 userVariables [
  User.bIsFrontEndPreview {type=NiagaraBool}
  User.FXScale {type=NiagaraFloat, defaultValue=1, description=""}
  User.Stage_01 {type=NiagaraBool, defaultValue=true}
  User.Stage_02 {type=NiagaraBool, defaultValue=true}
 ]
}
